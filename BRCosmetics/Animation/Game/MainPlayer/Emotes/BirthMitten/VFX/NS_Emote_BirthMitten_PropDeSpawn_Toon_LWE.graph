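# Niagara system for the toon-styled VFX of the BirthMitten emote's prop despawn.
# It combines airglow, sparkle, impact smoke and smoke emitters, with the glow and sparkles named as lightweight (LWE) variants; a 1m cube emitter is disabled.
# System-level logic pairs SystemState with a spawn-time variable assignment, using time-of-day color and float-array selection dynamic inputs.
# No user parameters are exposed.
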NiagaraSystem /BRCosmetics/Animation/Game/MainPlayer/Emotes/BirthMitten/VFX/NS_Emote_BirthMitten_PropDeSpawn_Toon_LWE
dependencies {
 usedDataInterfaces [&NiagaraDataInterfaceArrayFloat]
 usedDynamicInputs [&FNTimeOfDayColor, &SelectFloatFromArray]
 usedModules [
  &SystemState
  &self:SystemSpawnScript.SystemScriptSource.SystemScriptGraph.NiagaraNodeAssignment_0.SetVariables_#1
 ]
}
emitters {
 1mCube_Delete {bEnabled=false, rendererClasses [], renderers []}
 AirglowLWE {rendererClasses [], renderers []}
 SparklesLWE {rendererClasses [], renderers []}
 ToonImpactSmoke {rendererClasses [], renderers []}
 ToonSmoke {rendererClasses [], renderers []}
}
summary {
 systemName=NS_Emote_BirthMitten_PropDeSpawn_Toon_LWE
 emitters [
  {emitterName=AirglowLWE, rendererClasses []}
  {emitterName=SparklesLWE, rendererClasses []}
  {emitterName=ToonImpactSmoke, rendererClasses []}
  {emitterName=ToonSmoke, rendererClasses []}
  {bEnabled=false, emitterName=1mCube_Delete, rendererClasses []}
 ]
 userVariables []
}
