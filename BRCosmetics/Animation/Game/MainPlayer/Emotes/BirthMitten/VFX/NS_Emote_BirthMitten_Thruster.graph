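# Niagara system for the BirthMitten emote's thruster and light-ray VFX, built from burst-spawned particles that follow skeletal mesh bones or sockets.
# Flame emitters layer meshes, halo sprites and lights on two thruster points, pairing a brief ignition flare with long-lived jets.
# Delayed light rays start bright and stretched, then settle.
# Separate glow emitters average two sampled positions and switch intensity between gameplay and front end.
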NiagaraSystem /BRCosmetics/Animation/Game/MainPlayer/Emotes/BirthMitten/VFX/NS_Emote_BirthMitten_Thruster
dependencies {
 usedDataInterfaces [&NiagaraDataInterfaceCurve, &NiagaraDataInterfaceSkeletalMesh]
 usedDynamicInputs [&RandomRangeInt]
 usedModules [
  &self:SystemSpawnScript.SystemScriptSource.SystemScriptGraph.NiagaraNodeAssignment_0.SetVariables_#1
  &self:SystemSpawnScript.SystemScriptSource.SystemScriptGraph.NiagaraNodeAssignment_1.SetVariables_#2
  &SystemState
 ]
}
emitters {
 MeshRays {
  emitterUpdateScript [
   &EmitterState(
    Inactive Response=::NewEnumerator1
    Life Cycle Mode=::NewEnumerator1
    Loop Behavior=::NewEnumerator1
    Loop Duration Mode=::NewEnumerator1
    UseLoopDelay=false
   )
   &SpawnBurst_Instantaneous(Spawn Count=2, Spawn Time=1.5, Use Spawn Probability=false)
  ]
  particleSpawnScript [
   &InitializeParticle(
    Color Mode=::NewEnumerator0
    Lifetime=8.5
    Mesh Scale=<2 2.5 1>
    Mesh Scale Mode=::NewEnumerator4
    Position Mode=::NewEnumerator0
    Sprite Size Mode=::NewEnumerator3
    Uniform Sprite Size=60
    Write Material Random=true
   )
   &PartitionParticles()
   &InitialMeshOrientation(Mesh Orientation Mode=::NewEnumerator2, Orientation Vector=<0 1 0>, Use Rotation Vector=false)
  ]
  particleUpdateScript [
   &ParticleState()
   &ScaleMeshSize(Scale Factor=&Lerp_Vector(
     A=<1 1.75 1>
     B=<1 1 1>
     Alpha=&FloatFromCurve(FloatCurve=DataInterface() {propertyValues {Curve {keys [<at 0 0>,<at 0.2 1>,<at 1 1>]}}}, CurveIndex=$Particles.Age)
    ))
   &SkeletalMeshLocation(
    Bone / Socket Index=$Particles.PartitionParticles.Partition
    MaskSpawnGroup=false
    Mesh Sampling Type=::NewEnumerator5
    Skeletal Mesh=$System.Skeletal Mesh_LightRay
    Socket Sampling Mode=::NewEnumerator5
   )
   /SetVariables(
    Particles.Color_Corona:LinearColor=&Multiply_LinearColorByFloat(
     Linear Color=<rgba 0.1526 0.4358 0.5 1>
     Float=&FloatFromCurve(FloatCurve=DataInterface() {propertyValues {Curve {keys [<at 0 2>,<at 0.2 1>,<at 1 1>]}}}, CurveIndex=$Particles.Age)
    )
    Particles.Color_Halo:LinearColor=&Multiply_LinearColorByFloat(
     Linear Color=<rgba 0.092 0.2615 0.3 1>
     Float=&FloatFromCurve(FloatCurve=DataInterface() {propertyValues {Curve {keys [<at 0 2>,<at 0.2 1>,<at 1.75 1>]}}}, CurveIndex=$Particles.Age)
    )
    Particles.Color_MeshRay:LinearColor=&Multiply_LinearColorByFloat(
     Linear Color=<rgba 0.01 0.1784 0.2 1>
     Float=&FloatFromCurve(FloatCurve=DataInterface() {propertyValues {Curve {keys [<at 0 2.5>,<at 0.2 1>,<at 1 1>]}}}, CurveIndex=$Particles.Age)
    )
   ) {isSetParameters=true}
   &CameraOffset(Camera Offset Amount=20)
  ]
  rendererClasses [&NiagaraSpriteRendererProperties, &NiagaraMeshRendererProperties]
  renderers [
   {rendererClass=&NiagaraSpriteRendererProperties}
   {rendererClass=&NiagaraMeshRendererProperties, rendererIndex=1}
   {rendererClass=&NiagaraSpriteRendererProperties, rendererIndex=2}
  ]
 }
 NE_Flames {
  emitterUpdateScript [
   &EmitterState()
   &SpawnBurst_Instantaneous(Spawn Count=1, Spawn Time=1.5, Use Spawn Probability=false)
  ]
  particleSpawnScript [&InitializeParticle(
    Color Mode=::NewEnumerator0
    Lifetime=8.5
    Mesh Scale Mode=::NewEnumerator3
    Mesh Uniform Scale=25
    Position Mode=::NewEnumerator0
    Sprite Size Mode=::NewEnumerator3
    Uniform Sprite Size=140
    Write Material Random=true
   )]
  particleUpdateScript [
   &ParticleState()
   &SkeletalMeshLocation(
    MaskSpawnGroup=false
    Mesh Sampling Type=::NewEnumerator5
    Position Sampling=::NewEnumerator2
    Sampled Position Offset=&Lerp_Vector(
     A=<150 0 -20>
     B=<75 0 0>
     Alpha=&FloatFromCurve(FloatCurve=DataInterface() {propertyValues {Curve {keys [<at 0.75 0>,<at 0.84 1>]}}})
    )
    Skeletal Mesh=$System.Skeletal Mesh_LightRay
    Socket Sampling Mode=::NewEnumerator5
   )
   &SkeletalMeshLocation(
    Bone / Socket Index=1
    MaskSpawnGroup=false
    Mesh Sampling Type=::NewEnumerator5
    Position Sampling=::NewEnumerator2
    Sampled Position Offset=&Lerp_Vector(
     A=<150 0 -20>
     B=<75 0 0>
     Alpha=&FloatFromCurve(FloatCurve=DataInterface() {propertyValues {Curve {keys [<at 0.75 0>,<at 0.84 1>]}}})
    )
    Skeletal Mesh=$System.Skeletal Mesh_LightRay
    Socket Sampling Mode=::NewEnumerator5
   ) {module=SkeletalMeshLocation001}
   /SetVariables(Particles.Position_BoneLerp:NiagaraPosition=&Lerp_Vector(
     A=Linked() {linkedVariable {name=Particles.SkeletalMeshLocation.SampledPosition, type=NiagaraPosition}}
     B=Linked() {linkedVariable {name=Particles.SkeletalMeshLocation001.SampledPosition, type=NiagaraPosition}}
     Alpha=0.5
    )) {isSetParameters=true, module=SetVariables_#3}
   &ScaleColor(
    Scale Alpha=&FloatFromCurve(FloatCurve=DataInterface() {propertyValues {Curve {keys [<at 0 2.5>,<at 0.2 1>,<at 1 1>]}}}, CurveIndex=$Particles.Age)
    Scale RGB=&MakeVectorFromLinearColorRGB(Color=&Multiply_LinearColorByFloat(Linear Color=<rgba 0.5 0.977 1 1>, Float=&FN_FrontEndFloatModifier(Gameplay=8, Front End=5)))
    ScaleA=true
    ScaleRGB=true
   )
  ]
  rendererClasses [&NiagaraLightRendererProperties, &NiagaraMeshRendererProperties]
  renderers [
   {rendererClass=&NiagaraLightRendererProperties, rendererIndex=0}
   {rendererClass=&NiagaraMeshRendererProperties, rendererIndex=1}
  ]
 }
 NE_Flames002 {
  emitterUpdateScript [
   &EmitterState()
   &SpawnBurst_Instantaneous(Spawn Count=2, Spawn Time=0.1, Use Spawn Probability=false)
   /SetVariables(
    Emitter.Color_Jet:LinearColor=<rgba 5 5 5 1>
    Emitter.Color_Light:LinearColor=<rgba 2.6771 8.7152 10 1>
    Emitter.Color_SpriteHalo:LinearColor=<rgba 0 0.8887 1 1>
   ) {isSetParameters=true}
  ]
  particleSpawnScript [
   &InitializeParticle(
    Lifetime=10
    Mesh Scale=<0.2 0.2 0.2>
    Mesh Scale Mode=::NewEnumerator4
    Position Mode=::NewEnumerator0
    Sprite Size Mode=::NewEnumerator3
    Uniform Sprite Size=80
    Write Material Random=true
   )
   &PartitionParticles()
   &InitialMeshOrientation(Orientation Vector=<0 0 1>, Rotation=<0 0.25 0>, Use Orientation Vector=false, Use Rotation Vector=true) {module=InitialMeshOrientation001}
   /SetVariables(
    Particles.Scale_FlowSmooth:Vector3f=&Multiply_Vector(A=$Particles.Scale, B=<1 1 1.5>)
    Particles.Scale_HotCoreTurbulance:Vector3f=&Multiply_Vector(A=$Particles.Scale, B=<0.5 0.5 0.75>)
    Particles.Scale_Jet:Vector3f=&Multiply_Vector(A=$Particles.Scale, B=<0.3 0.3 0.6>)
   ) {isSetParameters=true, module=SetVariables_#4}
  ]
  particleUpdateScript [
   &ParticleState()
   &SkeletalMeshLocation(
    Bone / Socket Index=$Particles.PartitionParticles.Partition
    MaskSpawnGroup=false
    Mesh Sampling Type=::NewEnumerator5
    Sampled Position Offset=&MergeHelper_VectorSwitch(0=<0 0 0>, 1=<0 0 0>, 2=?, 3=?, Index=$Particles.PartitionParticles.Partition, Num=::NewEnumerator0)
    Skeletal Mesh=$System.Skeletal Mesh_Thruster
    Socket Sampling Mode=::NewEnumerator5
   )
   &ScaleColor(Scale RGB=&Multiply_VectorByFloat(Vector=<0 0.5 1>, Float=1), ScaleA=true, ScaleRGB=true)
  ]
  rendererClasses [
   &NiagaraSpriteRendererProperties
   &NiagaraMeshRendererProperties
   &NiagaraLightRendererProperties
  ]
  renderers [
   {rendererClass=&NiagaraSpriteRendererProperties}
   {rendererClass=&NiagaraMeshRendererProperties, rendererIndex=1}
   {rendererClass=&NiagaraMeshRendererProperties, rendererIndex=2}
   {rendererClass=&NiagaraMeshRendererProperties, rendererIndex=3}
   {rendererClass=&NiagaraLightRendererProperties, rendererIndex=4}
  ]
 }
 NE_Flames003 {
  emitterUpdateScript [
   &EmitterState()
   &SpawnBurst_Instantaneous(Spawn Count=1, Use Spawn Probability=false)
   &SpawnBurst_Instantaneous(Spawn Count=1, Spawn Time=0.1, Use Spawn Probability=false) {module=SpawnBurst_Instantaneous001}
   /SetVariables(
    Emitter.Color_FlowSmooth:LinearColor=&LerpLinearColors(StartColor=<rgba 2 2 2 1>, EndColor=<rgba 0 0 0 1>, LerpFactor=$StackContext.NormalizedLoopAge)
    Emitter.Color_HotCoreTurbulance:LinearColor=&LerpLinearColors(StartColor=<rgba 2 2 2 1>, EndColor=<rgba 0 0 0 1>, LerpFactor=$StackContext.NormalizedLoopAge)
    Emitter.Color_Jet:LinearColor=&LerpLinearColors(StartColor=<rgba 50 50 50 1>, EndColor=<rgba 0 0 0 1>, LerpFactor=$StackContext.NormalizedLoopAge)
    Emitter.Color_Light:LinearColor=&LerpLinearColors(StartColor=<rgba 9.7739 20 20 1>, EndColor=<rgba 0 0 0 1>, LerpFactor=$StackContext.NormalizedLoopAge)
    Emitter.Color_SpriteHalo:LinearColor=&LerpLinearColors(StartColor=<rgba 0.0336 2 2 1>, EndColor=<rgba 0 0 0 1>, LerpFactor=$StackContext.NormalizedLoopAge)
   ) {isSetParameters=true}
  ]
  particleSpawnScript [
   &InitializeParticle(
    Lifetime=0.2
    Mesh Scale=<0.2 0.2 0.2>
    Mesh Scale Mode=::NewEnumerator4
    Sprite Size Mode=::NewEnumerator3
    Uniform Sprite Size=140
    UsePositionOffset=false
    Write Material Random=true
   )
   &PartitionParticles()
   &InitialMeshOrientation(Orientation Vector=<0 0 1>, Rotation=<0 0.25 0>, Use Orientation Vector=false, Use Rotation Vector=true) {module=InitialMeshOrientation001}
   /SetVariables(
    Particles.Scale_FlowSmooth:Vector3f=&Multiply_Vector(A=$Particles.Scale, B=<1 1 1.5>)
    Particles.Scale_HotCoreTurbulance:Vector3f=&Multiply_Vector(A=$Particles.Scale, B=<0.5 0.5 0.75>)
    Particles.Scale_Jet:Vector3f=&Multiply_Vector(A=$Particles.Scale, B=<0.3 0.3 0.6>)
   ) {enabled=false, isSetParameters=true, module=SetVariables_#4}
  ]
  particleUpdateScript [
   &ParticleState()
   &SkeletalMeshLocation(
    Bone / Socket Index=$Particles.PartitionParticles.Partition
    MaskSpawnGroup=false
    Mesh Sampling Type=::NewEnumerator5
    Sampled Position Offset=&MergeHelper_VectorSwitch(0=<0 0 0>, 1=<0 0 0>, 2=?, 3=?, Index=$Particles.PartitionParticles.Partition, Num=::NewEnumerator0)
    Skeletal Mesh=$System.Skeletal Mesh_Thruster
    Socket Sampling Mode=::NewEnumerator5
   )
   /SetVariables(
    Particles.Scale_FlowSmooth:Vector3f=&Multiply_Vector(
     A=$Particles.Scale
     B=&Multiply_Vector(
      A=<1 1 1>
      B=/Niagara/DynamicInputs/Math/LerpMultipleVectors(
       Alpha:NiagaraFloat=$Particles.NormalizedAge
       V 1:Vector3f=<1 1 0>
       V 2:Vector3f=<1.4 1.4 3>
       V 3:Vector3f=<1 1 0>
       V 4:Vector3f=?
       V 5:Vector3f=?
       Vector Count:ENiagara_MultipleLerpCount=::NewEnumerator0
      )
     )
    )
    Particles.Scale_HotCoreTurbulance:Vector3f=&Multiply_Vector(
     A=$Particles.Scale
     B=&Multiply_Vector(
      A=<0.5 0.5 0.75>
      B=&LerpMultipleVectors(
       Alpha:NiagaraFloat=$Particles.NormalizedAge
       V 1:Vector3f=<1 1 0>
       V 2:Vector3f=<1.4 1.4 3>
       V 3:Vector3f=<1 1 0>
       V 4:Vector3f=?
       V 5:Vector3f=?
       Vector Count:ENiagara_MultipleLerpCount=::NewEnumerator0
      )
     )
    )
    Particles.Scale_Jet:Vector3f=&Multiply_Vector(
     A=$Particles.Scale
     B=&Multiply_Vector(
      A=<0.3 0.3 0.6>
      B=&LerpMultipleVectors(
       Alpha:NiagaraFloat=$Particles.NormalizedAge
       V 1:Vector3f=<1 1 0>
       V 2:Vector3f=<1.4 1.4 3>
       V 3:Vector3f=<1 1 0>
       V 4:Vector3f=?
       V 5:Vector3f=?
       Vector Count:ENiagara_MultipleLerpCount=::NewEnumerator0
      )
     )
    )
   ) {isSetParameters=true}
  ]
  rendererClasses [
   &NiagaraSpriteRendererProperties
   &NiagaraMeshRendererProperties
   &NiagaraLightRendererProperties
  ]
  renderers [
   {rendererClass=&NiagaraSpriteRendererProperties}
   {rendererClass=&NiagaraMeshRendererProperties, rendererIndex=1}
   {rendererClass=&NiagaraMeshRendererProperties, rendererIndex=2}
   {rendererClass=&NiagaraMeshRendererProperties, rendererIndex=3}
   {rendererClass=&NiagaraLightRendererProperties, rendererIndex=4}
  ]
 }
 NE_Flames004 {
  emitterUpdateScript [
   &EmitterState()
   &SpawnBurst_Instantaneous(Spawn Count=1, Use Spawn Probability=false)
  ]
  particleSpawnScript [&InitializeParticle(
    Color Mode=::NewEnumerator0
    Lifetime=10
    Mesh Scale Mode=::NewEnumerator3
    Mesh Uniform Scale=25
    Sprite Size Mode=::NewEnumerator3
    Uniform Sprite Size=140
    UsePositionOffset=false
    Write Material Random=true
   )]
  particleUpdateScript [
   &ParticleState()
   &SkeletalMeshLocation(
    MaskSpawnGroup=false
    Mesh Sampling Type=::NewEnumerator5
    Skeletal Mesh=$System.Skeletal Mesh_Thruster
    Socket Sampling Mode=::NewEnumerator5
   )
   &SkeletalMeshLocation(
    Bone / Socket Index=1
    MaskSpawnGroup=false
    Mesh Sampling Type=::NewEnumerator5
    Skeletal Mesh=$System.Skeletal Mesh
    Socket Sampling Mode=::NewEnumerator5
   ) {module=SkeletalMeshLocation001}
   /SetVariables(Particles.Position_BoneLerp:NiagaraPosition=&Lerp_Vector(
     A=Linked() {linkedVariable {name=Particles.SkeletalMeshLocation.SampledPosition, type=NiagaraPosition}}
     B=Linked() {linkedVariable {name=Particles.SkeletalMeshLocation001.SampledPosition, type=NiagaraPosition}}
     Alpha=0.5
    )) {isSetParameters=true, module=SetVariables_#3}
   &ScaleColor(
    Scale RGB=&MakeVectorFromLinearColorRGB(Color=&Multiply_LinearColorByFloat(
      Linear Color=<rgba 0.2307 0.9647 1 1>
      Float=&FN_FrontEndFloatModifier(Gameplay=10, Front End=25)
     ))
    ScaleA=true
    ScaleRGB=true
   )
   &ScaleColor(
    Scale Alpha=&Waveform("Waveform [1]"=::NewEnumerator2, "[1] Frequency"=20, "[1] Compound Frequency"=10, "[1] Amplitude Min/Max"=<0.5 1>)
    Scale RGB=&MakeVectorFromLinearColorRGB(Color=&Multiply_LinearColorByFloat(Linear Color=<rgba 0.2307 0.9647 1 1>, Float=&FN_FrontEndFloatModifier(Gameplay=10, Front End=5)))
    ScaleA=true
    ScaleRGB=true
   ) {enabled=false, module=ScaleColor001}
  ]
  rendererClasses [&NiagaraLightRendererProperties, &NiagaraMeshRendererProperties]
  renderers [
   {rendererClass=&NiagaraLightRendererProperties, rendererIndex=0}
   {rendererClass=&NiagaraMeshRendererProperties, rendererIndex=1}
  ]
 }
}
summary {
 systemName=NS_Emote_BirthMitten_Thruster
 emitters [
  {
   emitterName=NE_Flames002
   rendererClasses [
    &NiagaraSpriteRendererProperties
    &NiagaraMeshRendererProperties
    &NiagaraLightRendererProperties
   ]
  }
  {
   emitterName=NE_Flames003
   rendererClasses [
    &NiagaraSpriteRendererProperties
    &NiagaraMeshRendererProperties
    &NiagaraLightRendererProperties
   ]
  }
  {emitterName=NE_Flames004, rendererClasses [&NiagaraLightRendererProperties, &NiagaraMeshRendererProperties]}
  {emitterName=MeshRays, rendererClasses [&NiagaraSpriteRendererProperties, &NiagaraMeshRendererProperties]}
  {emitterName=NE_Flames, rendererClasses [&NiagaraLightRendererProperties, &NiagaraMeshRendererProperties]}
 ]
 userVariables [
  User.bIsFrontEnd {type=NiagaraBool}
  User.bIsFrontEndPreview {type=NiagaraBool}
 ]
}
